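# Anim Blueprint for the Hacivat backpack cosmetic that adds secondary physics motion to its bag, strings and bottle on top of an input pose.
# A serial run of AnimDynamics nodes in component space uses linear springs for the bag's top and bottom, then damped angular springs for five two-bone string chains and two bottle bones.
# Each node blends at partial alpha, and the bag nodes have an LOD threshold.
# The event graph holds no logic.
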AnimBlueprint /BRCosmetics/Accessories/FORT_Backpacks/M_MED_Hacivat_Backpack_01/Meshes/M_MED_Hacivat_Backpack_01_AnimBP
target_skeleton=&/M_MED_Hacivat_Backpack_01_Skeleton
graphs [
 AnimGraph {nodes [
   AnimGraphNode_SubInput_0 = Input Pose()
   n1 = Anim Dynamics - Bone: dyn_bag_top(ComponentPose=Local To Component(LocalPose=@AnimGraphNode_SubInput_0), Alpha=0.08, ExternalForce="0,0,0") {
    bGravityOverrideInSimSpace=true
    bLinearSpring=true
    bOverrideLinearDamping=true
    gravityScale=0.2
    lODThreshold=1
    linearDampingOverride=0.99
    linearSpringConstant=320
    alphaBoolBlend {blendOption=Linear}
    boundBone {boneName=dyn_bag_top}
    componentAppliedLinearAccClamp=<4000 4000 4000>
    componentLinearAccScale=<1 1 1>
    componentLinearVelScale=<1 1 1>
    physicsBodyDefinitions [{
      sphereCollisionRadius=0
      boundBone {boneName=dyn_bag_top}
      boxExtents=<15 35 10>
      constraintSetup {
       twistAxis=AxisZ
       angularLimitsMax=<0 0 0>
       angularLimitsMin=<0 0 0>
       angularTarget=<0 0 0>
       linearAxesMax=<10 10 10>
       linearAxesMin=<-10 -10 -10>
      }
      localJointOffset=<0 0 0>
     }]
   }
   n2 = Anim Dynamics - Bone: dyn_bag_bott(ComponentPose=@n1, Alpha=0.12, ExternalForce="0,0,0") {
    bGravityOverrideInSimSpace=true
    bLinearSpring=true
    bOverrideLinearDamping=true
    gravityScale=0.2
    lODThreshold=1
    linearDampingOverride=0.99
    linearSpringConstant=320
    alphaBoolBlend {blendOption=Linear}
    boundBone {boneName=dyn_bag_bott}
    componentAppliedLinearAccClamp=<4000 4000 4000>
    componentLinearAccScale=<1 1 1>
    componentLinearVelScale=<1 1 1>
    physicsBodyDefinitions [{
      sphereCollisionRadius=0
      boundBone {boneName=dyn_bag_bott}
      boxExtents=<30 25 15>
      constraintSetup {
       twistAxis=AxisZ
       angularLimitsMax=<0 0 0>
       angularLimitsMin=<0 0 0>
       angularTarget=<0 0 0>
       linearAxesMax=<10 10 10>
       linearAxesMin=<-10 -10 -10>
      }
      localJointOffset=<0 0 0>
     }]
   }
   n3 = "Anim Dynamics - Chain: dyn_string_top_01 -> dyn_string_top_02"(ComponentPose=@n2, Alpha=0.6, ExternalForce="0,0,0") {
    angularDampingOverride=0.99
    angularSpringConstant=500
    bAngularSpring=true
    bChain=true
    bGravityOverrideInSimSpace=true
    bOverrideAngularDamping=true
    numSolverIterationsPostUpdate=4
    numSolverIterationsPreUpdate=16
    alphaBoolBlend {blendOption=Linear}
    boundBone {boneName=dyn_string_top_01}
    chainEnd {boneName=dyn_string_top_02}
    componentAppliedLinearAccClamp=<4000 4000 4000>
    componentLinearAccScale=<1 1 1>
    componentLinearVelScale=<1 1 1>
    physicsBodyDefinitions [{
      sphereCollisionRadius=0
      boundBone {boneName=dyn_string_top_01}
      boxExtents=<20 20 10>
      constraintSetup {twistAxis=AxisZ, angularLimitsMax=<0 60 3>, angularLimitsMin=<-15 -60 -3>, angularTarget=<1 0 0>}
      localJointOffset=<0 0 -2>
     }]
   }
   n4 = "Anim Dynamics - Chain: dyn_string_mid_r_01 -> dyn_string_mid_r_02"(ComponentPose=@n3, Alpha=0.8, ExternalForce="0,0,0") {
    angularDampingOverride=0.9996
    angularSpringConstant=500
    bAngularSpring=true
    bChain=true
    bGravityOverrideInSimSpace=true
    bOverrideAngularDamping=true
    numSolverIterationsPostUpdate=4
    numSolverIterationsPreUpdate=16
    alphaBoolBlend {blendOption=Linear}
    boundBone {boneName=dyn_string_mid_r_01}
    chainEnd {boneName=dyn_string_mid_r_02}
    componentAppliedLinearAccClamp=<4000 4000 4000>
    componentLinearAccScale=<1 1 1>
    componentLinearVelScale=<1 1 1>
    physicsBodyDefinitions [{
      sphereCollisionRadius=0
      boundBone {boneName=dyn_string_mid_r_01}
      boxExtents=<9 10 6>
      constraintSetup {twistAxis=AxisZ, angularLimitsMax=<4 60 3>, angularLimitsMin=<-15 -60 -3>, angularTarget=<1 0 0>}
      localJointOffset=<0 0 -2>
     }]
   }
   n5 = "Anim Dynamics - Chain: dyn_string_mid_l_01 -> dyn_string_mid_l_02"(ComponentPose=@n4, Alpha=0.8, ExternalForce="0,0,0") {
    angularDampingOverride=0.9996
    angularSpringConstant=500
    bAngularSpring=true
    bChain=true
    bGravityOverrideInSimSpace=true
    bOverrideAngularDamping=true
    numSolverIterationsPostUpdate=4
    numSolverIterationsPreUpdate=16
    alphaBoolBlend {blendOption=Linear}
    boundBone {boneName=dyn_string_mid_l_01}
    chainEnd {boneName=dyn_string_mid_l_02}
    componentAppliedLinearAccClamp=<4000 4000 4000>
    componentLinearAccScale=<1 1 1>
    componentLinearVelScale=<1 1 1>
    physicsBodyDefinitions [{
      sphereCollisionRadius=0
      boundBone {boneName=dyn_string_mid_l_01}
      boxExtents=<12 15 4>
      constraintSetup {twistAxis=AxisZ, angularLimitsMax=<4 60 3>, angularLimitsMin=<-15 -60 -3>, angularTarget=<1 0 0>}
      localJointOffset=<0 0 -3>
     }]
   }
   n6 = Anim Dynamics - Bone: dyn_bottle_base(ComponentPose=@n5, Alpha=0.5, ExternalForce="0,0,0") {
    angularDampingOverride=0.99
    angularSpringConstant=5000
    bAngularSpring=true
    bGravityOverrideInSimSpace=true
    bOverrideAngularDamping=true
    numSolverIterationsPostUpdate=4
    numSolverIterationsPreUpdate=16
    alphaBoolBlend {blendOption=Linear}
    boundBone {boneName=dyn_bottle_base}
    physicsBodyDefinitions [{
      sphereCollisionRadius=0
      boundBone {boneName=dyn_bottle_base}
      boxExtents=<20 20 10>
      constraintSetup {twistAxis=AxisZ, angularLimitsMax=<3 5 2>, angularLimitsMin=<-30 -5 -2>, angularTarget=<1 0 0>}
      localJointOffset=<0 0 -3>
     }]
   }
   n7 = Anim Dynamics - Bone: dyn_bottle(ComponentPose=@n6, Alpha=0.5, ExternalForce="0,0,0") {
    angularDampingOverride=0.99
    angularSpringConstant=5000
    bAngularSpring=true
    bGravityOverrideInSimSpace=true
    bOverrideAngularDamping=true
    numSolverIterationsPostUpdate=4
    numSolverIterationsPreUpdate=16
    alphaBoolBlend {blendOption=Linear}
    boundBone {boneName=dyn_bottle}
    physicsBodyDefinitions [{
      sphereCollisionRadius=0
      boundBone {boneName=dyn_bottle}
      boxExtents=<20 20 10>
      constraintSetup {twistAxis=AxisZ, angularLimitsMax=<10 2 2>, angularLimitsMin=<-30 -2 -2>, angularTarget=<1 0 0>}
      localJointOffset=<0 0 -3>
     }]
   }
   n8 = "Anim Dynamics - Chain: dyn_string_bott_r_01 -> dyn_string_bott_r_02"(ComponentPose=@n7, Alpha=0.75, ExternalForce="0,0,0") {
    angularDampingOverride=0.9996
    angularSpringConstant=500
    bAngularSpring=true
    bChain=true
    bGravityOverrideInSimSpace=true
    bOverrideAngularDamping=true
    numSolverIterationsPostUpdate=4
    numSolverIterationsPreUpdate=16
    alphaBoolBlend {blendOption=Linear}
    boundBone {boneName=dyn_string_bott_r_01}
    chainEnd {boneName=dyn_string_bott_r_02}
    componentAppliedLinearAccClamp=<4000 4000 4000>
    componentLinearAccScale=<1 1 1>
    componentLinearVelScale=<1 1 1>
    physicsBodyDefinitions [{
      sphereCollisionRadius=0
      boundBone {boneName=dyn_string_bott_r_01}
      boxExtents=<15 8 4>
      constraintSetup {twistAxis=AxisZ, angularLimitsMax=<4 60 3>, angularLimitsMin=<-15 -60 -3>, angularTarget=<1 0 0>}
      localJointOffset=<0 0 -2>
     }]
   }
   n9 = "Anim Dynamics - Chain: dyn_string_bott_l_01 -> dyn_string_bott_l_02"(ComponentPose=@n8, Alpha=0.75, ExternalForce="0,0,0") {
    angularDampingOverride=0.9996
    angularSpringConstant=500
    bAngularSpring=true
    bChain=true
    bGravityOverrideInSimSpace=true
    bOverrideAngularDamping=true
    numSolverIterationsPostUpdate=4
    numSolverIterationsPreUpdate=16
    alphaBoolBlend {blendOption=Linear}
    boundBone {boneName=dyn_string_bott_l_01}
    chainEnd {boneName=dyn_string_bott_l_02}
    componentAppliedLinearAccClamp=<4000 4000 4000>
    componentLinearAccScale=<1 1 1>
    componentLinearVelScale=<1 1 1>
    physicsBodyDefinitions [{
      sphereCollisionRadius=0
      boundBone {boneName=dyn_string_bott_l_01}
      boxExtents=<18 20 9>
      constraintSetup {twistAxis=AxisZ, angularLimitsMax=<4 60 3>, angularLimitsMin=<-15 -60 -3>, angularTarget=<1 0 0>}
      localJointOffset=<0 0 -3>
     }]
   }
   Output Pose(Result=Component To Local(ComponentPose=@n9)) {layerGroup=None}
  ]}
 EventGraph {nodes [Event BlueprintUpdateAnimation()]}
]
variables []
